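# Niagara effect for a vault secure door slam, throwing a burst of smoke sprites from both sides of a line.
# Only Smoke_Burst001 is enabled; the spark and spiky flash emitters are disabled.
# Normalized execution index spaces particles along X and splits them into opposite Y velocities, then drag, curl noise and downward acceleration shape growing, fading puffs.
# Dynamic material parameters animate the smoke over lifetime and pass through the User.IsHolo switch.
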NiagaraSystem /MoonflaxMedallions/VFX/Vault/NS_MoonflaxMedallions_SecureDoorSlam
dependencies {
 usedDataInterfaces [&NiagaraDataInterfaceColorCurve, &NiagaraDataInterfaceCurve, &NiagaraDataInterfaceCamera]
 usedDynamicInputs []
 usedModules [&SystemState, &self:SetFootDustFXColorArrayIndex]
}
emitters {
 OmnidirectionalBurst {
  bEnabled=false
  emitterUpdateScript [
   &EmitterState()
   &SpawnBurst_Instantaneous(Spawn Count=12, Use Spawn Probability=false, UseLoopCountLimit=false)
  ]
  particleSpawnScript [
   /SetVariables(Particles.LongShortSparkFloat:NiagaraFloat=&Lerp_Float(
     A=0
     B=1
     Alpha=&FloatFromCurve(
      FloatCurve=DataInterface() {propertyValues {Curve {keys [<at 0 1>,<at 0.2 1>,<at 0.3 0>,<at 1 0>]}}}
      CurveIndex=&RandomRangeFloat(Minimum=0, Maximum=1)
     )
    )) {isSetParameters=true}
   &InitializeParticle(
    Color=<rgba 1 0.4054 0 1>
    Lifetime=&Lerp_Float(
     A=&RandomRangeFloat(Minimum=0.3, Maximum=0.6)
     B=&RandomRangeFloat(Minimum=1, Maximum=1.2)
     Alpha=$Particles.LongShortSparkFloat
    )
    Sprite Rotation Mode=::NewEnumerator2
    Sprite Size Mode=::NewEnumerator1
    Uniform Sprite Size Min=3
    UsePositionOffset=false
    Write Material Random=true
   )
   /SetVariables(Particles.Position=&AddVectorToPosition(
     Position=$Engine.Emitter.SimulationPosition
     Vector=&RandomRangeVector(Minimum=<-210 0 0>, Maximum=<210 1 1>)
    )) {isSetParameters=true}
   &AddVelocity(Velocity=&Multiply_Vector(
     A=&Lerp_Vector(
      A=&RandomRangeVector(Minimum=<1 350 150>, Maximum=<1 1250 250>)
      B=&RandomRangeVector(Minimum=<1 1550 550>, Maximum=<1 1950 850>)
      Alpha=$Particles.LongShortSparkFloat
     )
     B=&Lerp_Vector(
      A=<1 1 1>
      B=<1 -1 1>
      Alpha=/Niagara/DynamicInputs/Integer/V2/Round_Float(Float:NiagaraFloat=&ReturnNormalizedExecIndex())
     )
    )) {module=AddVelocity001}
  ]
  particleUpdateScript [
   &ParticleState()
   &Drag(Drag=&RandomRangeFloat(Minimum=12, Maximum=15), Use Linear Drag=true, Use Rotational Drag=false)
   &Color(Color=&ColorFromCurve(ColorCurve=DataInterface() {propertyValues {
       AlphaCurve {keys [<at 0 1 RCIM_Cubic>,<at 1 0 RCIM_Cubic>]}
       BlueCurve {keys [<at 0 0 RCIM_Cubic>,<at 0.7604 0 RCIM_Cubic>]}
       GreenCurve {keys [<at 0 3.77 RCIM_Cubic>,<at 0.7604 0.1797 RCIM_Cubic>]}
       RedCurve {keys [<at 0 10 RCIM_Cubic>,<at 0.7604 1 RCIM_Cubic>]}
      }}))
   &ScaleColor(
    Scale Alpha=&Multiply_Float(
     A=&FloatFromCurve(
      FloatCurve=DataInterface() {propertyValues {Curve {keys [<at -0.0046 1.0113 RCIM_Cubic>,<at 1.0011 0.002 RCIM_Cubic>]}}}
      Scale Curve=&FN_FrontEndFloatModifier(Gameplay=1, Front End=0.75)
     )
     B=&ClampFloat(Float=&Divide_Float(
       A=&Subtract_Float(
        A=&DistanceBetweenPositions(
         A=&ReturnCameraPosition(Coordinate Space:ENiagaraCoordinateSpace=::Simulation, Camera Query:NiagaraDataInterfaceCamera=DataInterface())
         B=$Particles.Position
        )
        B=100
       )
       B=100
      ))
    )
    ScaleA=true
    ScaleRGB=true
   )
   &SolveForcesAndVelocity(Clamp Velocity=false, Limit Acceleration=false)
   &ScaleSpriteSizeBySpeed(Max Scale Factor=<1 8>, Min Scale Factor=<0.2 0.01>)
   &ScaleSpriteSize(Scale Factor=&Vector2DFromFloat(Value=&FloatFromCurve(FloatCurve=DataInterface() {propertyValues {Curve {keys [
          <at -0.0046 1.0113 RCIM_Cubic>
          {arriveTangent=-2.2929, interpMode=RCIM_Cubic, leaveTangent=-2.2929, tangentMode=RCTM_User, time=1.0011, value=0.002}
         ]}}})))
  ]
  rendererClasses [&NiagaraSpriteRendererProperties]
  renderers [{rendererClass=&NiagaraSpriteRendererProperties}]
 }
 Smoke_Burst001 {
  emitterUpdateScript [
   &EmitterState(Life Cycle Mode=::NewEnumerator1, Loop Behavior=::NewEnumerator1, Loop Duration=2, UseLoopDelay=true)
   &SpawnBurst_Instantaneous(Spawn Count=16, Use Spawn Probability=false)
  ]
  particleSpawnScript [
   &InitializeParticle(
    Color=&UniformRangedLinearColor(Minimum=<rgba 1 1 1 0.8>, Maximum=<rgba 1 1 1 0.9>)
    Lifetime=&RandomRangeFloat(Minimum=0.5, Maximum=1)
    Sprite Rotation Mode=::NewEnumerator2
    Sprite Size Mode=::NewEnumerator1
    Uniform Sprite Size Max=200
    Uniform Sprite Size Min=140
    UsePositionOffset=false
    Write Material Random=true
   )
   /SetVariables(Particles.Position=&AddVectorToPosition(
     Position=$Engine.Emitter.SimulationPosition
     Vector=&Lerp_Vector(
      A=<-210 0 0>
      B=<210 0 0>
      Alpha=&ModuloFloat(A=&ReturnNormalizedExecIndex(Normalized Index Scale=2), B=1)
     )
    )) {isSetParameters=true}
   &AddVelocity(Velocity=&Lerp_Vector(
     A=&RandomRangeVector(Minimum=<150 350 0>, Maximum=<-150 750 1>)
     B=&RandomRangeVector(Minimum=<150 -350 0>, Maximum=<-150 -750 1>)
     Alpha=&Round_Float(Float:NiagaraFloat=&ReturnNormalizedExecIndex())
    ))
   /SetVariables(Particles.FirstFlipBookFrame:NiagaraInt32=&RandomRangeInt(Minimum=0, Maximum=64)) {isSetParameters=true}
  ]
  particleUpdateScript [
   &ParticleState()
   &Color(Color=&ColorFromCurve(ColorCurve=DataInterface() {propertyValues {
       AlphaCurve {keys [<at 0 0.2 RCIM_Cubic>,<at 1 0 RCIM_Cubic>]}
       BlueCurve {keys [<at 0.1279 0.1354>]}
       GreenCurve {keys [<at 0.1279 0.1354>]}
       RedCurve {keys [<at 0.1279 0.1354>]}
      }}))
   &Drag(
    Drag=&Multiply_Float(
     A=&RandomRangeFloat(Minimum=7, Maximum=8)
     B=&FloatFromCurve(FloatCurve=DataInterface() {propertyValues {Curve {keys [
          {arriveTangent=-2.2549, interpMode=RCIM_Cubic, leaveTangent=-2.2549, tangentMode=RCTM_User, time=0, value=1}
          {interpMode=RCIM_Cubic, tangentMode=RCTM_User, time=1, value=0}
         ]}}})
    )
    Use Linear Drag=true
    Use Rotational Drag=true
   )
   &DynamicMaterialParameters(
    Index 0 Param 1=&Add_Float(
     A=&RandomRangeFloat(Minimum=0, Maximum=1)
     B=&FloatFromCurve(
      FloatCurve=DataInterface() {propertyValues {Curve {keys [
          {arriveTangent=2.7533, interpMode=RCIM_Cubic, leaveTangent=2.847, tangentMode=RCTM_Break, time=0, value=0}
          {arriveTangent=1.3581, interpMode=RCIM_Cubic, leaveTangent=1.3581, tangentMode=RCTM_User, time=0.2558, value=0.5125}
          {arriveTangent=0.0529, interpMode=RCIM_Cubic, leaveTangent=0.2021, tangentMode=RCTM_User, time=1, value=1}
         ]}}}
      Scale Curve=0.2
     )
    )
    Index 0 Param 2=&MakeCustomFloatFromBool(Bool=$User.IsHolo, True Float=1, False Float=0)
    Index 0 Param 3=&FloatFromCurve(FloatCurve=DataInterface() {propertyValues {Curve {keys [{arriveTangent=-53.9199, time=0, value=0}, {leaveTangent=0.2723, time=0.25, value=0}]}}})
    Index 0 Param 4=&FloatFromCurve(FloatCurve=DataInterface() {propertyValues {Curve {keys [<at 0 1>,<at 1 0>]}}}, Scale Curve=0.2)
    Index 1 Param 1=&FloatFromCurve(FloatCurve=DataInterface() {propertyValues {Curve {keys [<at 0 0>, {arriveTangent=4.2257, interpMode=RCIM_Cubic, leaveTangent=4.2257, tangentMode=RCTM_User, time=0.25, value=0}]}}})
    Index 1 Param 2=&FloatFromCurve(FloatCurve=DataInterface() {propertyValues {Curve {keys [<at 0 0>]}}})
    Index 1 Param 3=&FloatFromCurve(FloatCurve=DataInterface() {propertyValues {Curve {keys [{interpMode=RCIM_Cubic, leaveTangent=6.3917, tangentMode=RCTM_Break, time=0, value=0}]}}})
    Index 1 Param 4=&FloatFromCurve(FloatCurve=DataInterface() {propertyValues {Curve {keys [{arriveTangent=-0.136, leaveTangent=-0.136, time=0, value=1}]}}})
    Index 2 Param 1=&FloatFromCurve(FloatCurve=DataInterface() {propertyValues {Curve {keys [
         {arriveTangent=2.646, interpMode=RCIM_Cubic, leaveTangent=2.646, tangentMode=RCTM_User, time=0, value=0}
         {leaveTangent=-0.0719, time=1, value=1}
        ]}}})
    Index 2 Param 2=&FloatFromCurve(
     FloatCurve=DataInterface() {propertyValues {Curve {keys [
         {arriveTangent=1.9056, time=0, value=0.2}
         {arriveTangent=0.0259, interpMode=RCIM_Cubic, leaveTangent=0.0259, tangentMode=RCTM_User, time=1, value=1.5}
        ]}}}
     Scale Curve=&RandomRangeFloat(Minimum=0.2, Maximum=0.3)
    )
    Index 2 Param 3=1
    Index 2 Param 4=&FloatFromCurve(
     FloatCurve=DataInterface() {propertyValues {Curve {keys [{arriveTangent=-0.6867, interpMode=RCIM_Cubic, leaveTangent=-0.6867, tangentMode=RCTM_User, time=0, value=1}]}}}
     Scale Curve=0.7
    )
    Module.Param0WriteEnabled=true
    Module.Param0WriteEnabled1=true
    Module.Param0WriteEnabled2=true
    Module.Param0WriteEnabled3=true
    Module.Param1WriteEnabled=true
    Module.Param1WriteEnabled1=true
    Module.Param1WriteEnabled2=true
    Module.Param1WriteEnabled3=true
    Module.Param2WriteEnabled=true
    Module.Param2WriteEnabled1=true
    Module.Param2WriteEnabled2=true
    Module.Param2WriteEnabled3=true
    Module.Param3WriteEnabled=true
    Module.Param3WriteEnabled1=true
    Module.Param3WriteEnabled2=true
    Module.Param3WriteEnabled3=true
    Write Parameter Index 1=true
    Write Parameter Index 2=true
    Write Parameter Index 3=true
   )
   &AccelerationForce(Acceleration=&RandomRangeVector(Minimum=<0 0 -200>, Maximum=<0 0 -405>))
   &CurlNoiseForce(Noise Strength=250, Pan Noise=false)
   &ScaleSpriteSize(Uniform Curve Sprite Scale=DataInterface() {propertyValues {Curve {keys [
        {arriveTangent=3.1806, interpMode=RCIM_Cubic, leaveTangent=3.1806, tangentMode=RCTM_User, time=0, value=0.2}
        {arriveTangent=0.8116, interpMode=RCIM_Cubic, leaveTangent=0.8116, tangentMode=RCTM_User, time=1, value=1.4}
       ]}}})
   &SolveForcesAndVelocity(Clamp Velocity=false, Limit Acceleration=false)
  ]
  rendererClasses [&NiagaraSpriteRendererProperties]
  renderers [
   {rendererClass=&NiagaraSpriteRendererProperties}
   {rendererClass=&NiagaraSpriteRendererProperties, rendererIndex=1}
  ]
 }
 Spiky {
  bEnabled=false
  emitterUpdateScript [
   &EmitterState(Life Cycle Mode=::NewEnumerator1, Loop Behavior=::NewEnumerator1, Loop Duration=2, UseLoopDelay=false)
   &SpawnBurst_Instantaneous(Spawn Count=6, Use Spawn Probability=false, UseLoopCountLimit=false)
  ]
  particleSpawnScript [
   &InitializeParticle(
    Color=&RandomRangeLinearColor(Minimum=<rgba 0.9948 0.9948 0.9948 1>, Maximum=<rgba 1 1 1 1>)
    Lifetime=&Multiply_Float(
     A=&RandomRangeFloat(Minimum=0.2, Maximum=0.25)
     B=&Lerp_Float(A=1, B=0.6, Alpha=$System.IsRunning)
    )
    Sprite Size=&RandomRangeVector2D(Minimum=<100 200>, Maximum=<150 250>)
    Sprite Size Mode=::NewEnumerator4
    UsePositionOffset=false
    Write Material Random=true
   )
   &PartitionParticles()
   /SetVariables(Particles.Position=&AddVectorToPosition(
     Position=$Engine.Emitter.SimulationPosition
     Vector=&Lerp_Vector(
      A=<-190 0 0>
      B=<190 0 0>
      Alpha=&ModuloFloat(A=&ReturnNormalizedExecIndex(), B=1)
     )
    )) {isSetParameters=true}
   &AddVelocity(Velocity=&Lerp_Vector(A=<0 -0.5 0.15>, B=<0 0.5 0.15>, Alpha=&MakeFloatFromInt(Int=$Particles.PartitionParticles.Partition)))
   &SubUVAnimation(End Frame=3, SubUV Animation Mode:ENiagara_SubUVLookupMode=::NewEnumerator2)
   /SetVariables(Particles.SpriteFacing=&Lerp_Vector(A=<0 0.5 1>, B=<0 -0.5 1>, Alpha=&MakeFloatFromInt(Int=$Particles.PartitionParticles.Partition))) {isSetParameters=true}
  ]
  particleUpdateScript [
   &ParticleState()
   &ScaleColor(
    Scale Alpha=&FloatFromCurve(
     FloatCurve=DataInterface() {propertyValues {Curve {keys [
         {arriveTangent=-1.375, interpMode=RCIM_Cubic, leaveTangent=-1.375, time=0.2, value=0.6}
         <at 0.6 0.05>
         {arriveTangent=-0.125, interpMode=RCIM_Cubic, leaveTangent=-0.125, time=1, value=0}
        ]}}}
     Scale Curve=0.25
    )
    Scale RGB=&MakeVectorFromLinearColorRGB(Color=&ColorFromCurve(ColorCurve=DataInterface() {propertyValues {
        AlphaCurve {keys [<at 0 1 RCIM_Cubic>,<at 1 0 RCIM_Cubic>]}
        BlueCurve {keys [
          {arriveTangent=1.0455, interpMode=RCIM_Cubic, time=0, value=0}
          <at 0.2244 0.5>
          {interpMode=RCIM_Cubic, leaveTangent=-1.3766, time=1, value=0}
         ]}
        GreenCurve {keys [
          {arriveTangent=0.8627, interpMode=RCIM_Cubic, time=0, value=0.3506}
          <at 0.2244 0.5>
          {interpMode=RCIM_Cubic, leaveTangent=-1.3766, time=1, value=0}
         ]}
        RedCurve {keys [<at 0 1 RCIM_Cubic>, <at 0.2244 0.5>, {interpMode=RCIM_Cubic, leaveTangent=-1.3766, time=1, value=0}]}
       }}))
    ScaleA=true
    ScaleRGB=true
   )
   &ScaleSpriteSize(Scale Factor=&Vector2DFromFloat(Value=&FloatFromCurve(FloatCurve=DataInterface() {propertyValues {Curve {keys [
          {arriveTangent=1.6317, interpMode=RCIM_Cubic, leaveTangent=1.6317, tangentMode=RCTM_User, time=0, value=0.1}
          {arriveTangent=0.2754, interpMode=RCIM_Cubic, leaveTangent=0.2754, tangentMode=RCTM_User, time=1.0463, value=1}
         ]}}})))
   &Drag(Drag=1.5, Use Linear Drag=true, Use Rotational Drag=false)
   &SolveForcesAndVelocity(Clamp Velocity=false, Limit Acceleration=false)
  ]
  rendererClasses [&NiagaraSpriteRendererProperties]
  renderers [{rendererClass=&NiagaraSpriteRendererProperties}]
 }
}
summary {
 systemName=NS_MoonflaxMedallions_SecureDoorSlam
 emitters [
  {emitterName=Smoke_Burst001, rendererClasses [&NiagaraSpriteRendererProperties]}
  {bEnabled=false, emitterName=Spiky, rendererClasses [&NiagaraSpriteRendererProperties]}
  {bEnabled=false, emitterName=OmnidirectionalBurst, rendererClasses [&NiagaraSpriteRendererProperties]}
 ]
 userVariables [
  User.bIsFrontEnd {type=NiagaraBool}
  User.bIsFrontEndPreview {type=NiagaraBool}
  User.IsHolo {type=NiagaraBool}
  User.SurfaceTypeIndex {type=NiagaraInt32}
 ]
}
